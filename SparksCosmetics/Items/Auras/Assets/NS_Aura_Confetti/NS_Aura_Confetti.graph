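# Niagara confetti aura cosmetic that fires colorful bursts from sockets on a skeletal mesh.
# Staggered spawn groups launch confetti, sparkles, smoke and mesh streamers from four sockets, while looping glow sprites track seven sockets.
# Confetti and Sparkles run on GPU, with near-duplicate CPU MedLow variants; only Confetti_MedLow reads the user color array.
# User flags adjust size, spread and speed for large bodies and brightness for front end, alongside time-of-day color scaling.
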NiagaraSystem /SparksCosmetics/Items/Auras/Assets/NS_Aura_Confetti/NS_Aura_Confetti
dependencies {
 usedDataInterfaces [
  &NiagaraDataInterfaceArrayFloat
  &NiagaraDataInterfaceArrayInt32
  &NiagaraDataInterfaceArrayFloat3
  &NiagaraDataInterfaceCurve
  &NiagaraDataInterfaceSpriteRendererInfo
  &NiagaraDataInterfaceColorCurve
  &NiagaraDataInterfaceSkeletalMesh
 ]
 usedDynamicInputs [&SimulationPosition]
 usedModules [
  &self:SystemSpawnScript.SystemScriptSource.SystemScriptGraph.NiagaraNodeAssignment_0.SetVariables_#1
  &SystemState
 ]
}
emitters {
 Confetti {
  simTarget=GPUComputeSim
  emitterUpdateScript [
   &EmitterState()
   &SpawnBurst_Instantaneous(Spawn Count=16, Use Spawn Probability=true, UseLoopCountLimit=false)
   &SpawnBurst_Instantaneous(Spawn Count=16, Spawn Group=1, Spawn Time=0.3, Use Spawn Probability=true, UseLoopCountLimit=false) {module=SpawnBurst_Instantaneous001}
   &SpawnBurst_Instantaneous(Spawn Count=16, Spawn Group=2, Spawn Time=0.6, Use Spawn Probability=true, UseLoopCountLimit=false) {module=SpawnBurst_Instantaneous002}
   &SpawnBurst_Instantaneous(Spawn Count=16, Spawn Group=3, Spawn Time=0.6, Use Spawn Probability=true, UseLoopCountLimit=false) {module=SpawnBurst_Instantaneous003}
  ]
  particleSpawnScript [
   &PartitionParticles(Number Of Partitions=3)
   &InitializeParticle(
    Color=&SelectLinearColorFromArray(
     Array Sampling Mode=::NewEnumerator0
     Direct Array Index=?
     Color Selection Array=DataInterface() {propertyValues {ColorData [<rgba 1 0.0203 0.6724 1>, <rgba 0.003 1 0.8963 1>, <rgba 0.4179 0.8388 0.0356 1>, <rgba 1 0.0262 0.1144 1>]}}
     Random Seed=0
    )
    Lifetime Mode=::NewEnumerator1
    Sprite Size Mode=::NewEnumerator1
    Uniform Sprite Size Max=6
    Uniform Sprite Size Min=3
    UsePositionOffset=false
    Write Material Random=true
   )
   &AddVelocity(
    Cone Angle=&Multiply_Float(A=60, B=&MakeCustomFloatFromBool(Bool=$User.IsLargeBody, True Float=1.25, False Float=1))
    Cone Axis=&SelectVectorFromArray(
     Direct Array Index=&SpawnGroup()
     Vector Selection Array=DataInterface() {propertyValues {FloatData [<0.33 0 1>, <-0.33 0 1>, <0.66 0 1>, <-0.66 0 1>]}}
    )
    Module.UseConeFalloff=true
    Module.UseVelDistribution=true
    Override Seed=false
    Velocity Mode=::NewEnumerator2
    Velocity Speed=&Multiply_Float(
     A=&RandomRangeFloat(Minimum=175, Maximum=325)
     B=&MakeCustomFloatFromBool(Bool=$User.IsLargeBody, True Float=0.8, False Float=1)
    )
   )
   &SkeletalMeshLocation(
    Bone / Socket Index=&SelectIntFromArray(Direct Array Index=&SpawnGroup(), Int Selection Array=DataInterface() {propertyValues {IntData [1,2,3,4]}})
    Bone Sampling Mode=::NewEnumerator4
    MaskSpawnGroup=false
    Skeletal Mesh=$System.Skeletal Mesh
   )
   /SetVariables(Particles.CurlSubImageFrame:NiagaraFloat=&RandomRangeFloat(Minimum=0, Maximum=15)) {isSetParameters=true, module=SetVariables_#2}
   /SetVariables(Particles.VisibilityTag=&MakeCustomIntFromBool(
     Bool=&RandomBool()
     True INT=&RandomRangeInt(Minimum=1, Maximum=2)
     False INT=0
    )) {isSetParameters=true, module=SetVariables_#3}
  ]
  particleUpdateScript [
   &ParticleState()
   &GravityForce(Gravity=<0 0 -111>)
   /Niagara/Modules/Update/Forces/CurlNoiseForce(Noise Frequency=0.02, Noise Strength=200, Pan Noise=true, Pan Noise Field=<0.05 0.05 0.05>)
   &Drag(
    Drag=&FloatFromCurve(FloatCurve=DataInterface() {propertyValues {Curve {keys [<at 0 0.5>,<at 1 1>]}}}, Scale Curve=8)
    Use Linear Drag=true
    Use Rotational Drag=true
   )
   &SolveForcesAndVelocity(Clamp Velocity=false, Limit Acceleration=false)
   &ScaleSpriteSize(Scale Factor=&Vector2DFromFloat(Value=&FloatFromCurve(FloatCurve=DataInterface() {propertyValues {Curve {keys [
          <at 0 0 RCIM_Cubic>
          {arriveTangent=-0.0623, interpMode=RCIM_Cubic, leaveTangent=-0.0623, tangentMode=RCTM_User, time=0.04, value=1}
          {arriveTangent=-0.0083, interpMode=RCIM_Cubic, leaveTangent=-0.0083, tangentMode=RCTM_User, time=0.5, value=1}
          <at 1 0 RCIM_Cubic>
         ]}}})))
   &SubUVAnimation(
    Random Start Frame=true
    "Sprite Renderer "=DataInterface() {propertyValues {SpriteRenderer=&self:Confetti_11.NiagaraSpriteRendererProperties_0}}
    Start Frame Offset:NiagaraInt32=&RandomRangeInt(Minimum=0, Maximum=555)
    UseEndFrame=false
    UseStartFrame=false
   )
   /SetVariables(Particles.CurlSpriteSize:Vector2f=&Multiply_Vector2D_ByFloat(Float=1.45, Vector2D=$Particles.SpriteSize)) {isSetParameters=true, module=SetVariables_#4}
   &ScaleColor(
    Scale RGB=&VectorFromFloat(Value=&FNTimeOfDayScalar(Night Scale=0.375, Morning Scale=0.4, Day Scale=0.625, Evening Scale=0.375))
    ScaleA=true
    ScaleRGB=true
   )
  ]
  rendererClasses [&NiagaraSpriteRendererProperties]
  renderers [
   {rendererClass=&NiagaraSpriteRendererProperties}
   {rendererClass=&NiagaraSpriteRendererProperties, rendererIndex=1}
   {rendererClass=&NiagaraSpriteRendererProperties, rendererIndex=2}
  ]
 }
 Confetti_MedLow {
  emitterUpdateScript [
   &EmitterState()
   &SpawnBurst_Instantaneous(Spawn Count=16, Use Spawn Probability=true, UseLoopCountLimit=false)
   &SpawnBurst_Instantaneous(Spawn Count=16, Spawn Group=1, Spawn Time=0.3, Use Spawn Probability=true, UseLoopCountLimit=false) {module=SpawnBurst_Instantaneous001}
   &SpawnBurst_Instantaneous(Spawn Count=16, Spawn Group=2, Spawn Time=0.6, Use Spawn Probability=true, UseLoopCountLimit=false) {module=SpawnBurst_Instantaneous002}
   &SpawnBurst_Instantaneous(Spawn Count=16, Spawn Group=3, Spawn Time=0.6, Use Spawn Probability=true, UseLoopCountLimit=false) {module=SpawnBurst_Instantaneous003}
  ]
  particleSpawnScript [
   &PartitionParticles(Number Of Partitions=3)
   &InitializeParticle(
    Color=&SelectLinearColorFromArray(Array Sampling Mode=::NewEnumerator0, Direct Array Index=?, Color Selection Array=$User.ConfettiColorArray, Random Seed=0)
    Lifetime Mode=::NewEnumerator1
    Sprite Size Mode=::NewEnumerator1
    Uniform Sprite Size Max=6
    Uniform Sprite Size Min=3
    UsePositionOffset=false
    Write Material Random=true
   )
   &AddVelocity(
    Cone Angle=&Multiply_Float(A=60, B=&MakeCustomFloatFromBool(Bool=$User.IsLargeBody, True Float=1.25, False Float=1))
    Cone Axis=&SelectVectorFromArray(
     Direct Array Index=&SpawnGroup()
     Vector Selection Array=DataInterface() {propertyValues {FloatData [<0.33 0 1>, <-0.33 0 1>, <0.66 0 1>, <-0.66 0 1>]}}
    )
    Module.UseConeFalloff=true
    Module.UseVelDistribution=true
    Override Seed=false
    Velocity Mode=::NewEnumerator2
    Velocity Speed=&Multiply_Float(
     A=&RandomRangeFloat(Minimum=175, Maximum=325)
     B=&MakeCustomFloatFromBool(Bool=$User.IsLargeBody, True Float=0.8, False Float=1)
    )
   )
   &SkeletalMeshLocation(
    Bone / Socket Index=&SelectIntFromArray(Direct Array Index=&SpawnGroup(), Int Selection Array=DataInterface() {propertyValues {IntData [1,2,3,4]}})
    Bone Sampling Mode=::NewEnumerator4
    MaskSpawnGroup=false
    Skeletal Mesh=$System.Skeletal Mesh
   )
   /SetVariables(Particles.CurlSubImageFrame:NiagaraFloat=&RandomRangeFloat(Minimum=0, Maximum=15)) {isSetParameters=true, module=SetVariables_#2}
   /SetVariables(Particles.VisibilityTag=&MakeCustomIntFromBool(
     Bool=&RandomBool()
     True INT=&RandomRangeInt(Minimum=1, Maximum=2)
     False INT=0
    )) {isSetParameters=true, module=SetVariables_#3}
  ]
  particleUpdateScript [
   &ParticleState()
   &GravityForce(Gravity=<0 0 -111>)
   &CurlNoiseForce(Noise Frequency=0.02, Noise Strength=200, Pan Noise=true, Pan Noise Field=<0.05 0.05 0.05>)
   &Drag(
    Drag=&FloatFromCurve(FloatCurve=DataInterface() {propertyValues {Curve {keys [<at 0 0.5>,<at 1 1>]}}}, Scale Curve=8)
    Use Linear Drag=true
    Use Rotational Drag=true
   )
   &SolveForcesAndVelocity(Clamp Velocity=false, Limit Acceleration=false)
   &ScaleSpriteSize(Scale Factor=&Vector2DFromFloat(Value=&FloatFromCurve(FloatCurve=DataInterface() {propertyValues {Curve {keys [
          <at 0 0 RCIM_Cubic>
          {arriveTangent=-0.0623, interpMode=RCIM_Cubic, leaveTangent=-0.0623, tangentMode=RCTM_User, time=0.04, value=1}
          {arriveTangent=-0.0083, interpMode=RCIM_Cubic, leaveTangent=-0.0083, tangentMode=RCTM_User, time=0.5, value=1}
          <at 1 0 RCIM_Cubic>
         ]}}})))
   &SubUVAnimation(
    Random Start Frame=true
    "Sprite Renderer "=DataInterface() {propertyValues {SpriteRenderer=&self:Confetti_MedLow_4.NiagaraSpriteRendererProperties_0}}
    Start Frame Offset:NiagaraInt32=&RandomRangeInt(Minimum=0, Maximum=555)
    UseEndFrame=false
    UseStartFrame=false
   )
   /SetVariables(Particles.CurlSpriteSize:Vector2f=&Multiply_Vector2D_ByFloat(Float=1.45, Vector2D=$Particles.SpriteSize)) {isSetParameters=true, module=SetVariables_#4}
   &ScaleColor(
    Scale RGB=&VectorFromFloat(Value=&FNTimeOfDayScalar(Night Scale=0.375, Morning Scale=0.4, Day Scale=0.625, Evening Scale=0.375))
    ScaleA=true
    ScaleRGB=true
   )
  ]
  rendererClasses [&NiagaraSpriteRendererProperties]
  renderers [
   {rendererClass=&NiagaraSpriteRendererProperties}
   {rendererClass=&NiagaraSpriteRendererProperties, rendererIndex=1}
   {rendererClass=&NiagaraSpriteRendererProperties, rendererIndex=2}
  ]
 }
 FireStreamers {
  emitterUpdateScript [
   &EmitterState()
   &SpawnBurst_Instantaneous(Use Spawn Probability=false, UseLoopCountLimit=false)
   &SpawnBurst_Instantaneous(Spawn Group=1, Spawn Time=0.25, Use Spawn Probability=false, UseLoopCountLimit=false) {module=SpawnBurst_Instantaneous001}
   &SpawnBurst_Instantaneous(Spawn Group=2, Spawn Time=0.55, Use Spawn Probability=false, UseLoopCountLimit=false) {module=SpawnBurst_Instantaneous002}
   &SpawnBurst_Instantaneous(Spawn Group=3, Spawn Time=0.55, Use Spawn Probability=false, UseLoopCountLimit=false) {module=SpawnBurst_Instantaneous003}
  ]
  particleSpawnScript [
   &PartitionParticles(Number Of Partitions=4, Particles Per Partition=1, Partition Mode:ENiagara_PartitionMode=::NewEnumerator1)
   &InitializeParticle(
    Color=&SelectLinearColorFromArray(
     Array Sampling Mode=::NewEnumerator1
     Direct Array Index=&SpawnGroup()
     Color Selection Array=DataInterface() {propertyValues {ColorData [<rgba 0.003 1 0.8963 1>, <rgba 0.4179 0.8388 0.0356 1>, <rgba 1 0.0203 0.6724 1>, <rgba 1 0.8126 0 1>]}}
     Random Seed=?
    )
    Mesh Scale=&MakeVector(
     X=1
     Y=1
     Z=&Multiply_Float(
      A=&SelectFloatFromArray(
       Direct Array Index=&SpawnGroup()
       Float Selection Array=DataInterface() {propertyValues {FloatData [1.5,1.5,1.25,1.25]}}
      )
      B=&MakeCustomFloatFromBool(Bool=$User.IsLargeBody, True Float=1.1, False Float=1)
     )
    )
    Mesh Scale Mode=::NewEnumerator4
    UsePositionOffset=false
    Write Material Random=true
   )
   &InitialMeshOrientation(Mesh Orientation Mode=::NewEnumerator1, Rotation=<0 0 0.3>, Use Rotation Vector=true)
  ]
  particleUpdateScript [
   &ParticleState()
   &DynamicMaterialParameters(
    Index 0 Param 1=&Multiply_Float(
     A=&FloatFromCurve(FloatCurve=DataInterface() {propertyValues {Curve {keys [
          {arriveTangent=-24.446, interpMode=RCIM_Cubic, leaveTangent=-24.446, tangentMode=RCTM_User, time=0, value=-20}
          {interpMode=RCIM_Cubic, leaveTangent=-25.1102, time=1, value=-30}
         ]}}})
     B=&MakeCustomFloatFromBool(Bool=$User.IsLargeBody, True Float=1.1, False Float=1)
    )
    Index 0 Param 2=&FloatFromCurve(FloatCurve=DataInterface() {propertyValues {Curve {keys [
         {arriveTangent=0.0333, interpMode=RCIM_Cubic, leaveTangent=0.0333, tangentMode=RCTM_User, time=0, value=0.05}
         {arriveTangent=2.4098, interpMode=RCIM_Cubic, leaveTangent=2.4098, tangentMode=RCTM_User, time=0.8, value=0.8}
        ]}}})
    Index 0 Param 3=&FloatFromCurve(FloatCurve=DataInterface() {propertyValues {Curve {keys [{arriveTangent=-0.2638, interpMode=RCIM_Cubic, time=0, value=0.1}, <at 0.5 0 RCIM_Cubic>]}}})
    Module.Param0WriteEnabled=true
    Module.Param1WriteEnabled=true
    Module.Param2WriteEnabled=true
    Module.Param3WriteEnabled=true
   )
   &SkeletalMeshLocation(
    Bone / Socket Index=&SelectIntFromArray(Direct Array Index=$Particles.PartitionParticles.Partition, Int Selection Array=DataInterface() {propertyValues {IntData [1,2,3,4]}})
    Bone Sampling Mode=::NewEnumerator4
    MaskSpawnGroup=false
    Orientation Sampling=::NewEnumerator1
    Sampled Position Offset=&SelectVectorFromArray(
     Direct Array Index=$Particles.PartitionParticles.Partition
     Vector Selection Array=DataInterface() {propertyValues {FloatData [<0 0 -10>, <0 0 -10>, <0 0 -5>, <0 0 -5>]}}
    )
    Sampled Position Offset Coordinate Space:ENiagaraSkelMeshTransforms=::NewEnumerator1
    Skeletal Mesh=$System.Skeletal Mesh
   )
   &UpdateMeshOrientation(
    Facing Direction=&SelectVectorFromArray(
     Direct Array Index=$Particles.PartitionParticles.Partition
     Vector Selection Array=DataInterface() {propertyValues {FloatData [<0.33 0 -0.15>, <-0.33 0 -0.15>, <0.66 0 -0.4>, <-0.66 0 -0.4>]}}
    )
    Facing Interpolation=false
    Orientation Method=::NewEnumerator1
    Override Quat=false
    TransformFacing=false
    TransformUp=false
   ) {module=UpdateMeshOrientation001}
   &ScaleMeshSize(Scale Factor=&MakeVector(
     X=1
     Y=1
     Z=&FloatFromCurve(FloatCurve=DataInterface() {propertyValues {Curve {keys [{arriveTangent=4.62, interpMode=RCIM_Cubic, leaveTangent=4.62, tangentMode=RCTM_User, time=0, value=0.3}, <at 0.4 1 RCIM_Cubic>]}}})
    ))
   &InterpolateOverTime(
    Rate Of Change=12
    Target Vector Value:Vector3f=&ConvertPositionToVector(Input Position=$Particles.Position, Convert To Absolute Large World Vector=false)
    Variable Type=::NewEnumerator2
   )
   &ScaleColor(
    Scale RGB=&Multiply_VectorByFloat(
     Vector=&VectorFromFloat(Value=&FNTimeOfDayScalar(
       Night Scale=0.25
       Morning Scale=0.33
       Day Scale=&MakeCustomFloatFromBool(Bool=$User.IsFrontEnd, True Float=0.25, False Float=0.3)
       Evening Scale=0.24
      ))
     Float=4
    )
    ScaleA=true
    ScaleRGB=true
   )
   /SetVariables(
    Particles.Erosion:NiagaraFloat=&Lerp_Float(A=0.1599, B=0.6, Alpha=$Particles.NormalizedAge)
    Particles.Interpolate Dynamic=&MakeFloat4FromVectorAndFloat("Vector (RGB)"=$StackContext.InterpolateOverTime.InterpolateOverTime_V3_.Moving Average, "Float (Alpha)"=$User.Interpolate Intensity)
   ) {isSetParameters=true}
  ]
  rendererClasses [&NiagaraMeshRendererProperties]
  renderers [{rendererClass=&NiagaraMeshRendererProperties, rendererIndex=0}]
 }
 Glow {
  emitterUpdateScript [
   &EmitterState(
    Inactive Response=::NewEnumerator1
    Life Cycle Mode=::NewEnumerator1
    Loop Behavior=::NewEnumerator1
    Loop Duration Mode=::NewEnumerator1
    UseLoopDelay=false
   )
   &SpawnBurst_Instantaneous(Spawn Count=7, Use Spawn Probability=false, UseLoopCountLimit=false)
  ]
  particleSpawnScript [
   &PartitionParticles(Number Of Partitions=7)
   &InitializeParticle(
    Color=&MakeCustomLinearColorFromBool(
     Bool=$User.IsFrontEnd
     True Color=<rgba 0.7552 0.0153 0.5078 1>
     False Color=&MultiplyLinearColor(Linear Color=<rgba 1 0.0203 0.6724 1>, Scale Factor=&MakeVector4_FromFloat(Float=1.25))
    )
    Lifetime=&RandomRangeFloat(Minimum=2, Maximum=3)
    Sprite Rotation Mode=::NewEnumerator0
    Sprite Size Mode=::NewEnumerator3
    Uniform Sprite Size=&Multiply_Float(
     A=&SelectFloatFromArray(
      Direct Array Index=&SpawnGroup()
      Float Selection Array=DataInterface() {propertyValues {FloatData [40,40,20,20,30,30]}}
     )
     B=&MakeCustomFloatFromBool(Bool=$User.IsLargeBody, True Float=1.5, False Float=1)
    )
    UsePositionOffset=false
    Write Material Random=true
   ) {module=InitializeParticle001}
   /SetVariables(Particles.glowScale:Vector2f=&Vector2DFromFloat(Value=&Multiply_Float(
      A=&SelectFloatFromArray(
       Direct Array Index=$Particles.PartitionParticles.Partition
       Float Selection Array=DataInterface() {propertyValues {FloatData [150,100,100,80,80,50,50]}}
      )
      B=&MakeCustomFloatFromBool(Bool=$User.IsLargeBody, True Float=1.2, False Float=1)
     ))) {isSetParameters=true}
   /SetVariables(Particles.glowCameraOffset=-10) {isSetParameters=true}
   /SetVariables(Particles.glowColor:LinearColor=&MultiplyLinearColor(
     Linear Color=&MultiplyLinearColor(Linear Color=<rgba 1 0.0203 0.6724 1>, Scale Factor=&MakeVector4_FromFloat(Float=3))
     Scale Factor=&MakeVector4_FromFloat(Float=&FNTimeOfDayScalar(
       Night Scale=0.25
       Morning Scale=0.2
       Day Scale=&MakeCustomFloatFromBool(Bool=$User.IsFrontEnd, True Float=0.2, False Float=0.26)
       Evening Scale=0.15
      ))
    )) {isSetParameters=true}
  ]
  particleUpdateScript [
   &ParticleState(
    Kill Particles When Lifetime Has Elapsed=false
    Let Infinitely Lived Particles Die When Emitter Deactivates=true
    Loop Particles Lifetime=true
   )
   &CameraOffset(Camera Offset Amount=5)
   &SpriteRotationRate(Rotation Rate=&RandomRangeFloat(Minimum=10, Maximum=20))
   &SkeletalMeshLocation(
    Bone / Socket Index=&SelectIntFromArray(Direct Array Index=$Particles.PartitionParticles.Partition, Int Selection Array=DataInterface() {propertyValues {IntData [0,1,2,3,4,5,6]}})
    Bone Sampling Mode=::NewEnumerator4
    MaskSpawnGroup=false
    Orientation Sampling=::NewEnumerator1
    Sampled Position Offset Coordinate Space:ENiagaraSkelMeshTransforms=::NewEnumerator1
    Skeletal Mesh=$System.Skeletal Mesh
   ) {module=SkeletalMeshLocation003}
   &DynamicMaterialParameters(
    Index 0 Param 1=&SelectFloatFromArray(Direct Array Index=$Particles.PartitionParticles.Partition, Float Selection Array=DataInterface() {propertyValues {FloatData [5,5,5,5,3,3]}})
    Index 0 Param 3=1
    Module.Param0WriteEnabled=true
    Module.Param1WriteEnabled=true
    Module.Param2WriteEnabled=true
    Module.Param3WriteEnabled=true
   )
   &ScaleColor(
    Scale RGB=&Multiply_VectorByFloat(
     Vector=&VectorFromFloat(Value=&FNTimeOfDayScalar(
       Night Scale=0.25
       Morning Scale=0.2
       Day Scale=&MakeCustomFloatFromBool(Bool=$User.IsFrontEnd, True Float=0.2, False Float=0.26)
       Evening Scale=0.15
      ))
     Float=4
    )
    ScaleA=true
    ScaleRGB=true
   )
  ]
  rendererClasses [&NiagaraSpriteRendererProperties]
  renderers [
   {rendererClass=&NiagaraSpriteRendererProperties}
   {rendererClass=&NiagaraSpriteRendererProperties, rendererIndex=1}
  ]
 }
 Smoke {
  emitterUpdateScript [
   &EmitterState()
   &SpawnBurst_Instantaneous(Spawn Count=10, Use Spawn Probability=false)
   &SpawnBurst_Instantaneous(Spawn Count=10, Spawn Group=1, Spawn Time=0.3, Use Spawn Probability=false) {module=SpawnBurst_Instantaneous001}
   &SpawnBurst_Instantaneous(Spawn Count=10, Spawn Group=3, Spawn Time=0.6, Use Spawn Probability=false) {module=SpawnBurst_Instantaneous002}
   &SpawnBurst_Instantaneous(Spawn Count=10, Spawn Group=2, Spawn Time=0.6, Use Spawn Probability=false) {module=SpawnBurst_Instantaneous003}
  ]
  particleSpawnScript [
   &InitializeParticle(
    Color=&SelectLinearColorFromArray(
     Array Sampling Mode=::NewEnumerator1
     Direct Array Index=&SpawnGroup()
     Color Selection Array=DataInterface() {propertyValues {ColorData [<rgba 0.3333 1 0.9306 1>, <rgba 0.5023 0.8388 0.1966 1>, <rgba 1 0.5208 0.8398 1>, <rgba 1 0.7428 0.1979 1>]}}
     Random Seed=?
    )
    Lifetime=&UniformRangedFloat(Minimum=0.25, Maximum=0.5)
    Mass Mode=::NewEnumerator1
    Sprite Rotation Angle=&UniformRangedFloat(Minimum=0, Maximum=360)
    Sprite Rotation Mode=::NewEnumerator0
    Sprite Size=&Vector2DFromFloat(Value=&UniformRangedFloat(Minimum=5, Maximum=10))
    Sprite Size Mode=::NewEnumerator4
    UsePositionOffset=false
    Write Material Random=true
   ) {module=InitializeParticle001}
   &AccelerationForce(Acceleration=&UniformRangedVector(Minimum=<0 0 0>, Maximum=<0 0 120>))
   &AddVelocity(Velocity=&UniformRangedVector(Minimum=<-50 -50 -50>, Maximum=<50 50 50>))
   &ApplyInitialForces()
   &SkeletalMeshLocation(
    Bone / Socket Index=&SelectIntFromArray(Direct Array Index=&SpawnGroup(), Int Selection Array=DataInterface() {propertyValues {IntData [1,2,3,4]}})
    Bone Sampling Mode=::NewEnumerator4
    MaskSpawnGroup=false
    Skeletal Mesh=$System.Skeletal Mesh
   )
  ]
  particleUpdateScript [
   &ParticleState()
   &Drag(Drag=3, Use Linear Drag=true, Use Rotational Drag=false)
   &SolveForcesAndVelocity(Clamp Velocity=false, Limit Acceleration=false)
   &ScaleSpriteSize(Scale Factor=&Vector2DFromFloat(Value=&FloatFromCurve(FloatCurve=DataInterface() {propertyValues {Curve {keys [
          {arriveTangent=1.491, interpMode=RCIM_Cubic, leaveTangent=1.491, tangentMode=RCTM_User, time=-0.0046, value=1.1969}
          <at 1.0011 3.4809 RCIM_Cubic>
         ]}}})))
   /SetVariables(Particles.SubImageIndex=&FloatFromCurve(FloatCurve=DataInterface() {propertyValues {Curve {keys [
         {arriveTangent=29.8623, interpMode=RCIM_Cubic, leaveTangent=29.8623, tangentMode=RCTM_User, time=-0.0046, value=5.0556}
         <at 1.0011 22.1689 RCIM_Cubic>
        ]}}})) {isSetParameters=true}
   &DynamicMaterialParameters(
    Index 0 Param 1=&FloatFromCurve(FloatCurve=DataInterface() {propertyValues {Curve {keys [<at -0.0046 1.0113 RCIM_Cubic>,<at 1.0011 0.002>]}}})
    Index 0 Param 4=&FloatFromCurve(FloatCurve=DataInterface() {propertyValues {Curve {keys [{arriveTangent=0.0079, interpMode=RCIM_Cubic, leaveTangent=0.0079, tangentMode=RCTM_User, time=0, value=0}, <at 1 0.5>]}}})
    Module.Param0WriteEnabled=true
    Module.Param1WriteEnabled=false
    Module.Param2WriteEnabled=false
    Module.Param3WriteEnabled=true
   ) {module=DynamicMatParams}
   &ScaleColor(
    Scale Alpha=&Multiply_Float(
     A=&FloatFromCurve(
      FloatCurve=DataInterface() {propertyValues {Curve {keys [
          {arriveTangent=-3.1757, interpMode=RCIM_Cubic, leaveTangent=-3.1757, tangentMode=RCTM_User, time=0.5, value=1.0113}
          {arriveTangent=-1.5121, interpMode=RCIM_Cubic, leaveTangent=-1.5121, tangentMode=RCTM_User, time=1.0011, value=0.002}
         ]}}}
      Scale Curve=0.5
     )
     B=&MakeCustomFloatFromBool(Bool=$User.IsFrontEnd, True Float=0.3, False Float=1)
    )
    Scale RGB=&Multiply_VectorByFloat(
     Vector=&VectorFromFloat(Value=&FNTimeOfDayScalar(Night Scale=0.175, Morning Scale=0.15, Day Scale=0.25, Evening Scale=0.2))
     Float=4
    )
    ScaleA=true
    ScaleRGB=true
   )
  ]
  rendererClasses [&NiagaraSpriteRendererProperties]
  renderers [{rendererClass=&NiagaraSpriteRendererProperties}]
 }
 Sparkles {
  simTarget=GPUComputeSim
  emitterUpdateScript [
   &EmitterState()
   &SpawnBurst_Instantaneous(Spawn Count=38, Spawn Time=0.1, Use Spawn Probability=false)
   &SpawnBurst_Instantaneous(Spawn Count=36, Spawn Group=1, Spawn Time=0.4, Use Spawn Probability=false) {module=SpawnBurst_Instantaneous001}
   &SpawnBurst_Instantaneous(Spawn Count=36, Spawn Group=2, Spawn Time=0.7, Use Spawn Probability=false) {module=SpawnBurst_Instantaneous002}
   &SpawnBurst_Instantaneous(Spawn Count=36, Spawn Group=3, Spawn Time=0.7, Use Spawn Probability=false) {module=SpawnBurst_Instantaneous003}
  ]
  particleSpawnScript [
   &InitializeParticle(
    Color=&HueShiftLinearColor(
     Hue=0
     Saturation=&RandomRangeFloat(Minimum=5, Maximum=10)
     Value=&RandomRangeFloat(Minimum=1, Maximum=10)
     Linear Color=&RandomRangeLinearColor(Minimum=<rgba 0 0 0 1>, Maximum=<rgba 1 1 1 1>, Color Channel Mode=::NewEnumerator2)
    )
    Lifetime Max=0.7
    Lifetime Min=0.2
    Lifetime Mode=::NewEnumerator1
    Position Mode=::NewEnumerator0
    Sprite Rotation Angle=&RandomRangeFloat(Minimum=0, Maximum=1)
    Sprite Rotation Mode=::NewEnumerator1
    Sprite Size Mode=::NewEnumerator1
    Uniform Sprite Size Max=6
    Uniform Sprite Size Min=1
    Write Material Random=true
   ) {module=InitializeParticle001}
   /Niagara/Modules/Spawn/Location/ConeLocation(Cone Angle=6, Cone Axis:Vector3f=<0 0 1>, Cone Length=30, MaskSpawnGroup=false, Module.Override Randomness=false)
   &AddVelocity(
    Cone Angle=&Multiply_Float(A=60, B=&MakeCustomFloatFromBool(Bool=$User.IsLargeBody, True Float=1.25, False Float=1))
    Cone Axis=&SelectVectorFromArray(
     Direct Array Index=&SpawnGroup()
     Vector Selection Array=DataInterface() {propertyValues {FloatData [<0.33 0 1>, <-0.33 0 1>, <0.66 0 1>, <-0.66 0 1>]}}
    )
    Module.UseConeFalloff=true
    Module.UseVelDistribution=true
    Override Seed=false
    Velocity Mode=::NewEnumerator2
    Velocity Speed=&Multiply_Float(
     A=&RandomRangeFloat(Minimum=150, Maximum=350)
     B=&MakeCustomFloatFromBool(Bool=$User.IsLargeBody, True Float=0.8, False Float=1)
    )
   )
   &SkeletalMeshLocation(
    Bone / Socket Index=&SelectIntFromArray(Direct Array Index=&SpawnGroup(), Int Selection Array=DataInterface() {propertyValues {IntData [1,2,3,4]}})
    Bone Sampling Mode=::NewEnumerator4
    MaskSpawnGroup=false
    Skeletal Mesh=$System.Skeletal Mesh
   ) {module=SkeletalMeshLocation001}
  ]
  particleUpdateScript [
   &ParticleState()
   &Drag(Drag=4, Use Linear Drag=true, Use Rotational Drag=false)
   &GravityForce(Gravity=&RandomRangeVector(Minimum=<0 0 -100>, Maximum=<1 1 -10>))
   &SolveForcesAndVelocity(Clamp Velocity=false, Limit Acceleration=false)
   &ScaleColor(
    Scale Alpha=&FloatFromCurve(FloatCurve=DataInterface() {propertyValues {Curve {keys [
         {arriveTangent=10.1697, interpMode=RCIM_Cubic, leaveTangent=10.1697, tangentMode=RCTM_User, time=0, value=0}
         <at 0.16 1 RCIM_Cubic>
         {arriveTangent=0.0209, interpMode=RCIM_Cubic, leaveTangent=0.0209, tangentMode=RCTM_User, time=0.35, value=0}
         <at 0.5 1 RCIM_Cubic>
         <at 0.7 0 RCIM_Cubic>
         <at 0.82 1 RCIM_Cubic>
         {arriveTangent=-8.68, interpMode=RCIM_Cubic, leaveTangent=-8.68, tangentMode=RCTM_User, time=1, value=0}
        ]}}})
    Scale RGB=&Multiply_VectorByFloat(
     Vector=&VectorFromFloat(Value=&FNTimeOfDayScalar(Night Scale=0.3, Morning Scale=0.4, Day Scale=0.5, Evening Scale=0.3))
     Float=&MakeCustomFloatFromBool(Bool=$User.IsFrontEnd, True Float=2, False Float=4)
    )
    ScaleA=true
    ScaleRGB=true
   )
   &ScaleSpriteSize(Scale Factor=&Vector2DFromFloat(Value=&FloatFromCurve(FloatCurve=DataInterface() {propertyValues {Curve {keys [
          {arriveTangent=-0.4596, interpMode=RCIM_Cubic, leaveTangent=-0.4596, tangentMode=RCTM_User, time=0.55, value=1}
          {arriveTangent=-4.9031, interpMode=RCIM_Cubic, leaveTangent=-4.9031, tangentMode=RCTM_User, time=1, value=0}
         ]}}})))
  ]
  rendererClasses [&NiagaraSpriteRendererProperties]
  renderers [{rendererClass=&NiagaraSpriteRendererProperties}]
 }
 Sparkles_MedLow {
  emitterUpdateScript [
   &EmitterState()
   &SpawnBurst_Instantaneous(Spawn Count=36, Spawn Time=0.1, Use Spawn Probability=false)
   &SpawnBurst_Instantaneous(Spawn Count=36, Spawn Group=1, Spawn Time=0.4, Use Spawn Probability=false) {module=SpawnBurst_Instantaneous001}
   &SpawnBurst_Instantaneous(Spawn Count=36, Spawn Group=2, Spawn Time=0.7, Use Spawn Probability=false) {module=SpawnBurst_Instantaneous002}
   &SpawnBurst_Instantaneous(Spawn Count=36, Spawn Group=3, Spawn Time=0.7, Use Spawn Probability=false) {module=SpawnBurst_Instantaneous003}
  ]
  particleSpawnScript [
   &InitializeParticle(
    Color=&HueShiftLinearColor(
     Hue=0
     Saturation=&RandomRangeFloat(Minimum=5, Maximum=10)
     Value=&RandomRangeFloat(Minimum=1, Maximum=10)
     Linear Color=&RandomRangeLinearColor(Minimum=<rgba 0 0 0 1>, Maximum=<rgba 1 1 1 1>, Color Channel Mode=::NewEnumerator2)
    )
    Lifetime Max=0.7
    Lifetime Min=0.2
    Lifetime Mode=::NewEnumerator1
    Position Mode=::NewEnumerator0
    Sprite Rotation Angle=&RandomRangeFloat(Minimum=0, Maximum=1)
    Sprite Rotation Mode=::NewEnumerator1
    Sprite Size Mode=::NewEnumerator1
    Uniform Sprite Size Max=6
    Uniform Sprite Size Min=1
    Write Material Random=true
   ) {module=InitializeParticle001}
   &ConeLocation(Cone Angle=6, Cone Axis:Vector3f=<0 0 1>, Cone Length=30, MaskSpawnGroup=false, Module.Override Randomness=false)
   &AddVelocity(
    Cone Angle=&Multiply_Float(A=60, B=&MakeCustomFloatFromBool(Bool=$User.IsLargeBody, True Float=1.25, False Float=1))
    Cone Axis=&SelectVectorFromArray(
     Direct Array Index=&SpawnGroup()
     Vector Selection Array=DataInterface() {propertyValues {FloatData [<0.33 0 1>, <-0.33 0 1>, <0.66 0 1>, <-0.66 0 1>]}}
    )
    Module.UseConeFalloff=true
    Module.UseVelDistribution=true
    Override Seed=false
    Velocity Mode=::NewEnumerator2
    Velocity Speed=&Multiply_Float(
     A=&RandomRangeFloat(Minimum=150, Maximum=350)
     B=&MakeCustomFloatFromBool(Bool=$User.IsLargeBody, True Float=0.8, False Float=1)
    )
   )
   &SkeletalMeshLocation(
    Bone / Socket Index=&SelectIntFromArray(Direct Array Index=&SpawnGroup(), Int Selection Array=DataInterface() {propertyValues {IntData [1,2,3,4]}})
    Bone Sampling Mode=::NewEnumerator4
    MaskSpawnGroup=false
    Skeletal Mesh=$System.Skeletal Mesh
   ) {module=SkeletalMeshLocation001}
  ]
  particleUpdateScript [
   &ParticleState()
   &Drag(Drag=4, Use Linear Drag=true, Use Rotational Drag=false)
   &GravityForce(Gravity=&RandomRangeVector(Minimum=<0 0 -100>, Maximum=<1 1 -10>))
   &SolveForcesAndVelocity(Clamp Velocity=false, Limit Acceleration=false)
   &ScaleColor(
    Scale Alpha=&FloatFromCurve(FloatCurve=DataInterface() {propertyValues {Curve {keys [
         {arriveTangent=10.1697, interpMode=RCIM_Cubic, leaveTangent=10.1697, tangentMode=RCTM_User, time=0, value=0}
         <at 0.16 1 RCIM_Cubic>
         {arriveTangent=0.0209, interpMode=RCIM_Cubic, leaveTangent=0.0209, tangentMode=RCTM_User, time=0.35, value=0}
         <at 0.5 1 RCIM_Cubic>
         <at 0.7 0 RCIM_Cubic>
         <at 0.82 1 RCIM_Cubic>
         {arriveTangent=-8.68, interpMode=RCIM_Cubic, leaveTangent=-8.68, tangentMode=RCTM_User, time=1, value=0}
        ]}}})
    Scale RGB=&Multiply_VectorByFloat(
     Vector=&VectorFromFloat(Value=&FNTimeOfDayScalar(Night Scale=0.3, Morning Scale=0.4, Day Scale=0.5, Evening Scale=0.3))
     Float=&MakeCustomFloatFromBool(Bool=$User.IsFrontEnd, True Float=2, False Float=4)
    )
    ScaleA=true
    ScaleRGB=true
   )
   &ScaleSpriteSize(Scale Factor=&Vector2DFromFloat(Value=&FloatFromCurve(FloatCurve=DataInterface() {propertyValues {Curve {keys [
          {arriveTangent=-0.4596, interpMode=RCIM_Cubic, leaveTangent=-0.4596, tangentMode=RCTM_User, time=0.55, value=1}
          {arriveTangent=-4.9031, interpMode=RCIM_Cubic, leaveTangent=-4.9031, tangentMode=RCTM_User, time=1, value=0}
         ]}}})))
  ]
  rendererClasses [&NiagaraSpriteRendererProperties]
  renderers [{rendererClass=&NiagaraSpriteRendererProperties}]
 }
}
summary {
 systemName=NS_Aura_Confetti
 emitters [
  {emitterName=Glow, rendererClasses [&NiagaraSpriteRendererProperties]}
  {emitterName=Confetti, simTarget=GPUComputeSim, rendererClasses [&NiagaraSpriteRendererProperties]}
  {emitterName=Sparkles, simTarget=GPUComputeSim, rendererClasses [&NiagaraSpriteRendererProperties]}
  {emitterName=Smoke, rendererClasses [&NiagaraSpriteRendererProperties]}
  {emitterName=FireStreamers, rendererClasses [&NiagaraMeshRendererProperties]}
  {emitterName=Confetti_MedLow, rendererClasses [&NiagaraSpriteRendererProperties]}
  {emitterName=Sparkles_MedLow, rendererClasses [&NiagaraSpriteRendererProperties]}
 ]
 userVariables [
  User.ConfettiColorArray {
   type=NiagaraDataInterfaceArrayColor
   description=""
   defaultValue {
    struct=NiagaraExt_VariableValue_DataInterface
    value {
     dataInterface=&self:NiagaraDataInterfaceArrayColor_0
     dataInterfaceClass=&NiagaraDataInterfaceArrayColor
    }
   }
  }
  User.Interpolate Intensity {type=NiagaraFloat, defaultValue=0.1, description=""}
  User.IsFrontEnd {type=NiagaraBool, defaultValue=true}
  User.IsLargeBody {type=NiagaraBool, defaultValue=true}
 ]
}
